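; Figment data layer toggle handler that switches World Partition data layers at runtime in response to routed request events.
; Once playlist data is ready, it listens for stateful requests, looks up the name in DataLayersMap and sets the requested state.
; It broadcasts a stateful loaded event immediately if the effective state matches.
; Otherwise the authority binds the manager's state-change delegate, which reverse-looks up the asset's name and broadcasts that event.
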(event EventBeginPlay
  (bind _datalayersmap (Variables|Default|GetDataLayersMap))
  (Class|FortAsyncActionCurrentPlaylistReady|PlaylistDataReadyAsync
    (:OnReady
      (bind (_asynctaskproxy _payload) (GameplayEventRouter|StartListeningToStatefulEvent (GameplayEventRouter|GetGlobalEventRouter) "/FigmentCore/Blueprints/FigmentEvent_DataLayerRequestStructure.FigmentEvent_DataLayerRequestStructure")
        (:OnEventReceived
          _)
        (:OnSavedEventStateExists
          _))
      (bind _returnvalue (WorldPartition|GetDataLayerManager))
      (bind (_payload_datalayername_2_490b9361452c80f00289faa03211106c _payload_state_9_4c4c82a240527be7983ea7b252ecbd7c) (Utilities|Struct|BreakFigmentEventDataLayerRequestStructure _payload))
      (if (.ReturnValue (Utilities|Map|Find _datalayersmap _payload_datalayername_2_490b9361452c80f00289faa03211106c))
        (DataLayers|SetDataLayerInstanceRuntimeState _returnvalue (DataLayers|GetDataLayerInstancefromAsset _returnvalue (Utilities|Map|Find _datalayersmap _payload_datalayername_2_490b9361452c80f00289faa03211106c)) _payload_state_9_4c4c82a240527be7983ea7b252ecbd7c false)
        (bind _returnvalue_2 (WorldPartition|GetDataLayerManager))
        (if (Utilities|Enum|Equal(Enum) (DataLayers|GetDataLayerInstanceEffectiveRuntimeState _returnvalue_2 (DataLayers|GetDataLayerInstancefromAsset _returnvalue (Utilities|Map|Find _datalayersmap _payload_datalayername_2_490b9361452c80f00289faa03211106c))) _payload_state_9_4c4c82a240527be7983ea7b252ecbd7c)
          (GameplayEventRouter|BroadcastStatefulEvent (GameplayEventRouter|GetGlobalEventRouter) (Utilities|Struct|MakeFigmentEventDataLayerLoadedStructure _payload_datalayername_2_490b9361452c80f00289faa03211106c _payload_state_9_4c4c82a240527be7983ea7b252ecbd7c) "/FigmentCore/Blueprints/FigmentEvent_DataLayerLoadedStructure.FigmentEvent_DataLayerLoadedStructure")
          (else
            (Networking|SwitchHasAuthority
              (:Authority
                (EventDispatchers|BindEventtoOnDataLayerInstanceRuntimeStateChanged _returnvalue_2 (delegate Custom|CustomEvent))))))))))

(event Custom|CustomEvent
  (params (DataLayer :type Data%20Layer%20Instance%20Object%20Reference) (State :type EDataLayerRuntimeState%20Enum :default "Unloaded"))
  (bind _returnvalue (DataLayer|GetAsset DataLayer))
  (Utilities|IsValid _returnvalue
    (:Is%20Valid
      (bind (_keypin _valuepin) (Utilities|Map|ForEachLoop(Map) (Variables|Default|GetDataLayersMap)))
      (Utilities|IsValid _valuepin
        (:Is%20Valid
          (if (== _valuepin _returnvalue)
            (GameplayEventRouter|BroadcastStatefulEvent (GameplayEventRouter|GetGlobalEventRouter) (Utilities|Struct|MakeFigmentEventDataLayerLoadedStructure _keypin State) "/FigmentCore/Blueprints/FigmentEvent_DataLayerLoadedStructure.FigmentEvent_DataLayerLoadedStructure")))))))

(fn ConstructionScript)
