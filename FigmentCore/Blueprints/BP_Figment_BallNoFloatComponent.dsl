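; Component for Fortnite physics balls that, as its name suggests, keeps a ball from floating after its support is destroyed.
; When enabled, it runs its sleep check on the server once playlist data is ready.
; The check line-traces around the ball's center of mass, printing hit classes and tracking non-ignored actors.
; It re-enables ball ticking when the tracked actor is destroyed, or immediately if none is valid.
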(event Custom|UnbindTracedActor
  (bind _tracedactor (Variables|Default|GetTracedActor)
    (:then
      (Game|UnbindEventfromOnDestroyed _tracedactor (EventDispatchers|CreateEvent :!event Custom|TracedActorDestroyed))
      (Variables|Default|SetTracedActor 0))
    (:else)))

(event Custom|OnBallSleep
  (bind _tracedactor (Variables|Default|GetTracedActor))
  (CallFunction|UnbindTracedActor)
  (bind _outhits (Collision|MultiLineTraceForObjects (+ (Physics|GetCenterOfMass (Variables|/Script/FortniteGame.FortPhysicsBall|GetSphereCollision (Variables|Default|GetComponentOwner)) "None") "0.000000,0.000000,15.000000") (- (Physics|GetCenterOfMass (Variables|/Script/FortniteGame.FortPhysicsBall|GetSphereCollision (Variables|Default|GetComponentOwner)) "None") "0.000000,0.000000,40.000000") (Utilities|Array|MakeArray "ObjectTypeQuery1")))
  (bind _output_get (Variables|Default|SetOut%20Hits _outhits))
  (bind _array_element (Utilities|Array|ForEachLoopwithBreak _output_get)
    (:LoopBody
      (Development|PrintString (Utilities|GetClassDisplayName (Utilities|GetClass (.HitActor (Collision|BreakHitResult _array_element)))))
      (if (not (Utilities|Set|ContainsItem (Variables|Default|GetActorTypesToIgnore) (Utilities|GetClass (.HitActor (Collision|BreakHitResult _array_element)))))
        (Variables|Default|SetTracedActor (.HitActor (Collision|BreakHitResult _array_element)))))
    (:Completed
      _))
  (Utilities|IsValid _tracedactor
    (:Is%20Valid
      (Game|BindEventtoOnDestroyed _tracedactor (delegate Custom|TracedActorDestroyed)))
    (:Is%20Not%20Valid
      (Class|PhysicsBallMaster|SetTickingEnabled (Variables|Default|GetComponentOwner) true))))

(event EventTick
  (params (DeltaSeconds :type Float%20%28single-precision%29)))

(event Custom|TracedActorDestroyed
  (params (DestroyedActor :type Actor%20Object%20Reference))
  (Class|PhysicsBallMaster|SetTickingEnabled (Variables|Default|GetComponentOwner) true))

(event EventBeginPlay
  (Class|FortAsyncActionCurrentPlaylistReady|PlaylistDataReadyAsync
    (:OnReady
      (if (ScalableFloat|GetValueasBool (Variables|Default|GetComponentEnabled))
        (Networking|SwitchHasAuthority
          (:Authority
            (bind _asphysics_ball_master (Utilities|Casting|CastToPhysicsBall_Master (Components|GetOwner)))
            (Variables|Default|SetComponentOwner _asphysics_ball_master)
            (CallFunction|OnBallSleep)))))))
