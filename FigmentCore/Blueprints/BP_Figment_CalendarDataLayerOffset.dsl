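; BP_Figment_CalendarDataLayerOffset switches World Partition data layers during time windows offset from a calendar event's start.
; It gathers event names from data table rows in its offset list, then re-evaluates matching entries when events activate or end.
; Inside the window, enabled layers are Activated and disabled layers Unloaded; outside, the states invert.
; Its CanUpdateInMatch override logs and refuses in-match updates for performance reasons, and state handling logs heavily.
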(event EventBeginPlay)

(event Collision|EventActorBeginOverlap
  (params (OtherActor :type Actor%20Object%20Reference)))

(event EventTick
  (params (DeltaSeconds :type Float%20%28single-precision%29)))

(fn CanUpdateInMatch
  (if (|CallParentFunction)
    (Class|BuildingFunctionLibrary|LogString "NewEnumerator5" "Attempting to update a Data Layer Calendar Event during a match.  This is not supported due to performance concerns." true self)
    (return false)))

(fn CheckEventActiveWithOffset
  (params (Event%20Name :type String) (Start%20Time%20Offset :type Figment%20Calendar%20Time%20Structure%20Structure) (End%20Time%20Offset :type Figment%20Calendar%20Time%20Structure%20Structure) (Event%20Row%20Name :type String))
  (bind _startoffset_timespan (LocalVariables|GetStartOffset_Timespan))
  (bind _endoffset_timespan (LocalVariables|GetEndOffset_Timespan))
  (bind (_start_time_offset_days_2_4de00cd0416ae18ae6ba4b884fe78a9e _start_time_offset_hours_4_209fe35442361442ce31c4a483631aa9 _start_time_offset_minutes_6_7ec92a7b4f78ca97e742c6b681e237fe) (Utilities|Struct|BreakFigmentCalendarTimeStructure (Variables|GetStart%20Time%20Offset_Days_2_4DE00CD0416AE18AE6BA4B884FE78A9E)))
  (bind (_end_time_offset_days_2_4de00cd0416ae18ae6ba4b884fe78a9e _end_time_offset_hours_4_209fe35442361442ce31c4a483631aa9 _end_time_offset_minutes_6_7ec92a7b4f78ca97e742c6b681e237fe) (Utilities|Struct|BreakFigmentCalendarTimeStructure (Variables|GetEnd%20Time%20Offset_Days_2_4DE00CD0416AE18AE6BA4B884FE78A9E)))
  (Variables|SetStartOffset_Timespan_Days (Math|Timespan|MakeTimespan _start_time_offset_days_2_4de00cd0416ae18ae6ba4b884fe78a9e _start_time_offset_hours_4_209fe35442361442ce31c4a483631aa9 _start_time_offset_minutes_6_7ec92a7b4f78ca97e742c6b681e237fe))
  (Variables|SetEndOffset_Timespan_Days (Math|Timespan|MakeTimespan _end_time_offset_days_2_4de00cd0416ae18ae6ba4b884fe78a9e _end_time_offset_hours_4_209fe35442361442ce31c4a483631aa9 _end_time_offset_minutes_6_7ec92a7b4f78ca97e742c6b681e237fe))
  (bind (_biseventactive _timesincebegin) (CalendarEvents|GetCalendarEventInformation Event%20Name))
  (if (and (>= _timesincebegin _startoffset_timespan) (< _timesincebegin _endoffset_timespan))
    (CallFunction|LogIt (Utilities|String|ToString(Text) (Utilities|Text|FormatText "NSLOCTEXT(\"[1ED943FF7B908B50B846D5B139E090D1]\", \"820D064D465DD5C918071BA39DC8CD32\", \"{a} : BP_CalendarDataLayerOffset:CheckEventActiveWithOffset - Calendar Event Offset: {b}[{c}] is active \")" (Variables|Default|GetDisplayName) (Variables|GetEvent%20Row%20Name) (Variables|GetEvent%20Name))))
    (else
      (CallFunction|LogIt (Utilities|String|ToString(Text) (Utilities|Text|FormatText "NSLOCTEXT(\"[1ED943FF7B908B50B846D5B139E090D1]\", \"EAC1C6BC46791F723F2C238C4706106A\", \"{a} : BP_CalendarDataLayerOffset:CheckEventActiveWithOffset - Calendar Event Offset: {b}[{c}]  is NOT active\")" (Variables|Default|GetDisplayName) (Variables|GetEvent%20Row%20Name) (Variables|GetEvent%20Name))))))
  (return (and (>= _timesincebegin _startoffset_timespan) (< _timesincebegin _endoffset_timespan))))

(fn GetUTCNow
  (bind (_returnvalue_year _returnvalue_month _returnvalue_day) (Math|DateTime|BreakDateTime (Math|DateTime|UTCNow)))
  (return _returnvalue_year _returnvalue_month _returnvalue_day 0 0 0 0))

(fn GetEventNames
  (for _array_element (Variables|Default|GetCalendarDataLayerOffsetList)
    (bind _array_element_event_15_b977ce844d8c1b1578942680debf71d6 (Utilities|Struct|BreakFigmentCalendarOffsetData _array_element))
    (bind (_array_element_event_15_b977ce844d8c1b1578942680debf71d6_datatable _array_element_event_15_b977ce844d8c1b1578942680debf71d6_rowname) (Utilities|Struct|BreakDataTableRowHandle _array_element_event_15_b977ce844d8c1b1578942680debf71d6))
    (bind _returnvalue (Utilities|GetDataTableRow _array_element_event_15_b977ce844d8c1b1578942680debf71d6_datatable _array_element_event_15_b977ce844d8c1b1578942680debf71d6_rowname))
    (if (not (Utilities|String|IsEmpty (Utilities|Struct|BreakFigmentCalendarOffsetStructure _returnvalue)))
      (Utilities|Array|AddUnique (LocalVariables|GetLocalEventNames) (Utilities|Struct|BreakFigmentCalendarOffsetStructure _returnvalue))))
  (return (LocalVariables|GetLocalEventNames)))

(fn HandleEventActive
  (params (EventName :type String))
  (CallFunction|HandleEventStateChanged EventName true))

(fn HandleEventEnded
  (params (EventName :type String))
  (CallFunction|HandleEventStateChanged EventName))

(fn HandleEventStateChanged
  (params (EventName :type String) (bIsEventActive :type Boolean))
  (bind _iseventactrivewithoffset (LocalVariables|GetIsEventActriveWithOffset))
  (bind _returnvalue (select _iseventactrivewithoffset "Unloaded" "Activated"))
  (bind _iseventactrivewithoffset_2 (LocalVariables|GetIsEventActriveWithOffset))
  (bind _returnvalue_2 (select _iseventactrivewithoffset_2 "Activated" "Unloaded"))
  (bind _current_row_name (LocalVariables|GetCurrent%20Row%20Name))
  (bind _displayname (Variables|Default|GetDisplayName))
  (for _array_element (Variables|Default|GetCalendarDataLayerOffsetList)
    (bind (_array_element_event_15_b977ce844d8c1b1578942680debf71d6 _array_element_enableddatalayers_10_cd008dd745929054ce3abea949e9e2d7 _array_element_disableddatalayers_12_f4743f674010f28ef86c679b34843b22) (Utilities|Struct|BreakFigmentCalendarOffsetData _array_element))
    (bind (_array_element_event_15_b977ce844d8c1b1578942680debf71d6_datatable _array_element_event_15_b977ce844d8c1b1578942680debf71d6_rowname) (Utilities|Struct|BreakDataTableRowHandle _array_element_event_15_b977ce844d8c1b1578942680debf71d6))
    (LocalVariables|SetCurrent%20Row%20Name (Utilities|String|ToString(Name) _array_element_event_15_b977ce844d8c1b1578942680debf71d6_rowname))
    (bind _returnvalue_3 (Utilities|GetDataTableRow _array_element_event_15_b977ce844d8c1b1578942680debf71d6_datatable _array_element_event_15_b977ce844d8c1b1578942680debf71d6_rowname)
      (:then
        (CallFunction|LogIt (Utilities|String|ToString(Text) (Utilities|Text|FormatText "NSLOCTEXT(\"[1ED943FF7B908B50B846D5B139E090D1]\", \"9EBE742246AE3DC60EAD84AC0C70577E\", \"{a} : BP_Figment_CalendarDataLayerOffset:HandleEventActive - Calendar Event Offset: {b} was found \")" _displayname (Utilities|String|ToString(Name) _array_element_event_15_b977ce844d8c1b1578942680debf71d6_rowname))))
        (if (Utilities|String|EqualExactly(String) (Variables|GetEventName) (Utilities|Struct|BreakFigmentCalendarOffsetStructure _returnvalue_3))
          (CallFunction|LogIt (Utilities|String|ToString(Text) (Utilities|Text|FormatText "NSLOCTEXT(\"[1ED943FF7B908B50B846D5B139E090D1]\", \"E5DC50AE4B99FC56BED7898C19566EF0\", \"{a} : BP_Figment_CalendarDataLayerOffset:HandleEventActive - Calendar Event Name: {b}[{c}] was found.\")" (Variables|Default|GetDisplayName) _current_row_name (Variables|GetEventName))))
          (bind _isactive (CallFunction|CheckEventActivewithOffset (Variables|GetEventName) (.StartOffset_16_468C72414F6F194F74A505B2CD53A10B (Utilities|Struct|BreakFigmentCalendarOffsetStructure _returnvalue_3)) (.EndOffset_18_4FFA481F462094F591788E8A38D39705 (Utilities|Struct|BreakFigmentCalendarOffsetStructure _returnvalue_3)) (LocalVariables|GetCurrent%20Row%20Name)))
          (LocalVariables|SetIsEventActriveWithOffset _isactive)
          (bind _returnvalue_4 (WorldPartition|GetDataLayerManager))
          (CallFunction|LogIt (Utilities|String|ToString(Text) (Utilities|Text|FormatText "NSLOCTEXT(\"[1ED943FF7B908B50B846D5B139E090D1]\", \"804874844A51C5EACB53B8A8F6434BC4\", \"{a} : BP_Figment_CalendarDataLayerOffset:HandleEventActive - Getting Data Layer Manager : {b}\")" (Variables|Default|GetDisplayName) _returnvalue_4)))
          (Utilities|FlowControl|Sequence
            (:then_0
              (for _array_element_2 _array_element_enableddatalayers_10_cd008dd745929054ce3abea949e9e2d7
                (DataLayers|SetDataLayerInstanceRuntimeState _returnvalue_4 (DataLayers|GetDataLayerInstancefromAsset _returnvalue_4 _array_element_2) _returnvalue_2 false)
                (CallFunction|LogIt (Utilities|String|ToString(Text) (Utilities|Text|FormatText "NSLOCTEXT(\"[1ED943FF7B908B50B846D5B139E090D1]\", \"2356FB4F435914F7E7E255A37649702E\", \"{a} : BP_Figment_CalendarDataLayerOffset:HandleEventActive:ENABLED Layers  from the Enable Layer- Calling Set Data Layer Instance Runtime State with the following information:\\r\\nDataLayer Manager: {b}\\r\\nDataLayer Instance: {c}\\r\\nSetting to State: {d}\\r\\nBecause Event is: {e}\")" (Variables|Default|GetDisplayName) _returnvalue_4 (DataLayers|GetDataLayerInstancefromAsset _returnvalue_4 _array_element_2) (Utilities|String|EnumtoString _returnvalue_2) (LocalVariables|GetIsEventActriveWithOffset))))))
            (:then_1
              (for _array_element_3 _array_element_disableddatalayers_12_f4743f674010f28ef86c679b34843b22
                (DataLayers|SetDataLayerInstanceRuntimeState _returnvalue_4 (DataLayers|GetDataLayerInstancefromAsset _returnvalue_4 _array_element_3) _returnvalue false)
                (CallFunction|LogIt (Utilities|String|ToString(Text) (Utilities|Text|FormatText "NSLOCTEXT(\"[1ED943FF7B908B50B846D5B139E090D1]\", \"15C9CF0B4E81697E38EF75AE2CA44410\", \"{a} : BP_Figment_CalendarDataLayerOffset:HandleEventActive:DISABLED Layers from the Disable Array - Calling Set Data Layer Instance Runtime State with the following information:\\r\\nDataLayer Manager: {b}\\r\\nDataLayer Instance: {c}\\r\\nSetting to State: {d}\\r\\nBecause Event is: {e}\")" (Variables|Default|GetDisplayName) _returnvalue_4 (DataLayers|GetDataLayerInstancefromAsset _returnvalue_4 _array_element_3) (Utilities|String|EnumtoString _returnvalue) (LocalVariables|GetIsEventActriveWithOffset)))))))
          (else
            (CallFunction|LogIt (Utilities|String|ToString(Text) (Utilities|Text|FormatText "NSLOCTEXT(\"[1ED943FF7B908B50B846D5B139E090D1]\", \"A58AF1794E695D0968D0F1A24BD486C8\", \"{a} : BP_Figment_CalendarDataLayerOffset:HandleEventActive - Calendar Event Name: {b}[{c}] was NOT found.\")" (Variables|Default|GetDisplayName) _current_row_name (Variables|GetEventName))) true))))
      (:RowNotFound
        (CallFunction|LogIt (Utilities|String|ToString(Text) (Utilities|Text|FormatText "NSLOCTEXT(\"[1ED943FF7B908B50B846D5B139E090D1]\", \"83C0CF124102D727927D18BA2C41004E\", \"{a} : BP_Figment_CalendarDataLayerOffset:HandleEventActive - Calendar Event Offset: {b} was NOT found \")" _displayname (Utilities|String|ToString(Name) _array_element_event_15_b977ce844d8c1b1578942680debf71d6_rowname))))))))

(fn ConstructionScript
  (|CallParentFunction))
