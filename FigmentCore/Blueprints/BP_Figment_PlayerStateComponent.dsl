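; Player state component for Figment that selects the classic elimination animation and manages the Icy Feet gameplay effects.
; On BeginPlay it adds the death animation tag to its owner and starts async waits for the Icy Feet tag, which GAB_SurfaceChange manages.
; Tag addition applies the Icy Feet effect and stores its handle.
; Tag removal clears that effect, applies a linger effect and toggles the trigger-once flag.
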(event EventBeginPlay
  (bind _returnvalue (Components|GetOwner))
  ;[1 1 1] Use the classic elimination animation
  (Ability|GameplayEffect|AddLooseGameplayTags (Components|GetOwner) (GameplayTags|MakeGameplayTagContainerfromTag (Variables|Default|GetDeathAnimationGameplayTag)) true)
  ; Apply Icy Feet GEs based on when the Gameplay Tag is added/removed (which is managed in GAB_SurfaceChange).
  (Ability|Tasks|WaitGameplayTagAddToActor _returnvalue (Variables|Default|GetIcy%20Feet%20Gameplay%20Tag)
    (:then
      (Ability|Tasks|WaitGameplayTagRemoveFromActor _returnvalue (Variables|Default|GetIcy%20Feet%20Gameplay%20Tag) (Variables|Default|GetOnly%20Trigger%20Once)
        (:Removed
          (CallFunction|RemoveIcyFeetGE _returnvalue))))
    (:Added
      (CallFunction|ApplyIcyFeetGE _returnvalue))))

(event EventTick
  (params (DeltaSeconds :type Float%20%28single-precision%29)))

(fn ApplyIcyFeetGE
  (params (Owner%20Actor :type Actor%20Object%20Reference))
  (bind _returnvalue (GameplayEffects|ApplyGameplayEffectToSelf :self (Ability|GetAbilitySystemComponent Owner%20Actor) :GameplayEffectClass (Variables|Default|GetIcy%20Feet%20GE) :Level 0.0))
  (Variables|Default|SetIcy%20Feet%20GE%20Handle _returnvalue)
  (Variables|Default|SetOnly%20Trigger%20Once false))

(fn RemoveIcyFeetGE
  (params (Owner%20Actor :type Actor%20Object%20Reference))
  (bind _returnvalue (Ability|GetAbilitySystemComponent Owner%20Actor))
  (GameplayEffects|RemoveActiveGameplayEffect _returnvalue (Variables|Default|GetIcy%20Feet%20GE%20Handle) -1)
  (GameplayEffects|ApplyGameplayEffectToSelf :self _returnvalue :GameplayEffectClass (Variables|Default|GetIcy%20Feet%20Linger%20GE) :Level 0.0)
  (Variables|Default|SetOnly%20Trigger%20Once true))
