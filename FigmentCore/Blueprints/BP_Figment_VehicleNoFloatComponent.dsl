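; Vehicle component that stops sleeping Fortnite vehicles from floating when the actor beneath them is destroyed.
; On authority, once playlist data is ready and if enabled, it binds to sleep state changes and runs its check.
; The check traces down from the center of mass, debug-prints hit classes and stores the last non-ignored actor.
; It moves its OnDestroyed binding to that actor, whose destruction wakes the sleeping vehicle.
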(event Custom|UnbindTracedActor
  (bind _tracedactor (Variables|Default|GetTracedActor)
    (:then
      (Game|UnbindEventfromOnDestroyed _tracedactor (EventDispatchers|CreateEvent :!event Custom|TracedActorDestroyed)))
    (:else)))

(event Custom|OnVehicleSleep
  (bind _tracedactor (Variables|Default|GetTracedActor))
  (CallFunction|UnbindTracedActor)
  (bind _outhits (Collision|MultiLineTraceForObjects (Physics|GetCenterOfMass (Variables|/Script/FortniteGame.FortAthenaVehicle|GetSkeletalMesh (Variables|Default|GetComponentOwner)) "None") (- (Physics|GetCenterOfMass (Variables|/Script/FortniteGame.FortAthenaVehicle|GetSkeletalMesh (Variables|Default|GetComponentOwner)) "None") "0.000000,0.000000,200.000000") (Utilities|Array|MakeArray "ObjectTypeQuery1")))
  (bind _output_get (Variables|Default|SetOut%20Hits _outhits))
  (bind _array_element (Utilities|Array|ForEachLoopwithBreak _output_get)
    (:LoopBody
      (Development|PrintString (Utilities|GetClassDisplayName (Utilities|GetClass (.HitActor (Collision|BreakHitResult _array_element)))))
      (if (not (Utilities|Set|ContainsItem (Variables|Default|GetActorTypesToIgnore) (Utilities|GetClass (.HitActor (Collision|BreakHitResult _array_element)))))
        (Variables|Default|SetTracedActor (.HitActor (Collision|BreakHitResult _array_element)))))
    (:Completed
      _))
  (Utilities|IsValid _tracedactor
    (:Is%20Valid
      (Game|BindEventtoOnDestroyed _tracedactor (delegate Custom|TracedActorDestroyed)))))

(event Custom|TracedActorDestroyed
  (params (DestroyedActor :type Actor%20Object%20Reference))
  (bind _componentowner (Variables|Default|GetComponentOwner))
  (if (Vehicle|IsAsleep _componentowner)
    (Vehicle|WakeUp _componentowner)))

(event EventBeginPlay
  (Class|FortAsyncActionCurrentPlaylistReady|PlaylistDataReadyAsync
    (:OnReady
      (if (ScalableFloat|GetValueasBool (Variables|Default|GetComponentEnabled))
        (Networking|SwitchHasAuthority
          (:Authority
            (bind _asfort_athena_vehicle (Utilities|Casting|CastToFortAthenaVehicle (Components|GetOwner)))
            (bind _output_get (Variables|Default|SetComponentOwner _asfort_athena_vehicle))
            (Vehicle|BindEventtoOnSleepStateChangedDelegate _output_get (delegate Custom|ComponentOwnerSleepStateChanged))
            (CallFunction|OnVehicleSleep)))))))

(event EventTick
  (params (DeltaSeconds :type Float%20%28single-precision%29)))

(event Custom|ComponentOwnerSleepStateChanged
  (params (bIsAwake :type Boolean))
  (if (not bIsAwake)
    (CallFunction|OnVehicleSleep)))
